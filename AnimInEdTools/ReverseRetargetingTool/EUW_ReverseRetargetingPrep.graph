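# Editor utility widget that batch reverse-retargets animation sequences for Fortnite characters and companions, picked in an embedded Animation Browser or the Content Browser.
# The selected asset and variant resolve an AnimatableCreature and item definition, which drive source and target meshes and output path.
# A custom name-matched or default Biped/Quadruped IK Retargeter feeds DuplicateAndRetarget.
# Results get a ReverseRetarget curve, and the Content Browser syncs to their folder.
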EditorUtilityWidgetBlueprint /AnimInEdTools/ReverseRetargetingTool/EUW_ReverseRetargetingPrep
graphs [
 AnimatableCreature Find {nodes [
   n1 = AnimatableCreature Find!() -> (then, AnimatableCreatures:Array of Animatable Creature Object References, Asset:String, Variant:String) {is=title K2Node_FunctionEntry}
   > n2 = Branch(Condition=Is Not Empty!(TargetArray:Array of Animatable Creature Object References=@n1) -> (ReturnValue:bool))
   > n3 = Branch(Condition="Equal, Case Insensitive (String)"(A:String=TrimTrailing!(SourceString:String=@n1.Variant) -> (ReturnValue:String), B:String=All) -> (ReturnValue:bool=~) {is=title K2Node_PromotableOperator})
   > LogString(InString=Finding AC through All)
   > Set Best Fit(Best Fit:Animatable Creature Object Reference) -> (Output_Get:Animatable Creature Object Reference)
   > n4 = For Each Loop with Break!(Exec:Exec, Array:Array of Animatable Creature Object References=@n1, Break:Exec) -> (LoopBody:Exec, Array Element:Animatable Creature Object Reference, Array Index:int, Completed:Exec)
   Return Node!(
    execute=@n4.Completed
    AnimatableCreature:Animatable Creature Object Reference=Get Best Fit() -> (Best Fit:Animatable Creature Object Reference)
    Exact Match:bool=false
   )
   > Branch(Condition="Equal Exactly (String)"(A=Get Asset!(self:Animatable Creature Object Reference=@n4) -> (Asset:String), B=Get Asset!() -> (Asset)))
   n5 = Get Variant!(self:Animatable Creature Object Reference=@n4) -> (Variant:String)
   > n6 = Branch(Condition=OR Boolean(A="Equal, Case Insensitive (String)"(A=@n5, B=Shared) {is=title K2Node_CallFunction}, B="Equal, Case Insensitive (String)"(A=@n5, B=Core)))
   > Return Node(AnimatableCreature=@n4, Exact Match=false)
   Set Best Fit(execute=@n6.else, Best Fit=@n4)
   LogString(execute=@n3.else, InString=Append!(A="Asset: ", B=@n1.Asset, C:String=" Variant: ", D:String=@n1.Variant) -> (ReturnValue))
   > n7 = For Each Loop(Array:Array of Animatable Creature Object References=@n1) -> (Array Element:Animatable Creature Object Reference)
   Return Node(execute=@n7.Completed, Exact Match=false)
   n8 = Get Variant(self=@n7)
   n9 = Get Variant!() -> (Variant)
   n10 = Get Asset!(self:Animatable Creature Object Reference=@n7) -> (Asset)
   n11 = Get Asset!() -> (Asset)
   > n12 = Branch(Condition=AND Boolean(A="Equal, Case Insensitive (String)"(A=@n11, B=@n10), B="Equal, Case Insensitive (String)"(A=@n8, B=@n9)))
   > Return Node(AnimatableCreature=@n7, Exact Match=true)
   LogString(execute=@n12.else, InString=Append(A="Asset: ", B=@n10, C=" Variant: ", D=@n9))
   LogString(execute=@n2.else, InString=Empty AC List)
   > Return Node(Exact Match=false)
  ]}
 Cache IKRs {nodes [
   Cache IKRs() {is=title K2Node_FunctionEntry}
   > n1 = GetAssetsByClass!(
    execute
    self:Asset Registry Interface=GetAssetRegistry()
    ClassPathName:Top Level Asset Path Structure=GetSoftClassTopLevelAssetPath!(SoftClassReference:Object Soft Class Reference=/Script/IKRig.IKRetargeter) -> (ReturnValue:Top Level Asset Path Structure)
    bSearchSubClasses:bool=false
   ) -> (then, OutAssetData:Array of Asset Data Structures, ReturnValue:bool)
   > n2 = For Each Loop(Array:Array of Asset Data Structures=@n1) -> (Array Element:Asset Data Structure)
   Return Node!(execute=@n2.Completed)
   > Add!(
    execute
    TargetMap:Map of Strings to Soft Object Path Structures=Get Cached IKRs() -> (Cached IKRs:Map of Strings to Soft Object Path Structures)
    Key:"String (by ref)"="To String (Name)"(InName=Break Asset Data(AssetData=@n2).AssetName)
    Value:"Soft Object Path Structure (by ref)"=ToSoftObjectPath!(InAssetData:"Asset Data Structure (by ref)"=@n2) -> (ReturnValue:Soft Object Path Structure)
   ) -> (then) {is=title K2Node_CallFunction}
  ]}
 EventGraph {nodes [
   "On Clicked (ExecButton)"() {is=title K2Node_ComponentBoundEvent}
   > n1 = GetSelectedItems!(execute, self:List View Object Reference=Get Anim List View() -> (Anim List View:List View Object Reference)) -> (then, Items:Array of Object References, ReturnValue:bool)
   > n2 = Branch(Condition=Is Not Empty(TargetArray:Array of Object References=@n1))
   n3 = Get!(execute:Exec=[@n2.then, @n8.Yes]) -> (then:Exec, else:Exec, CID:Athena Character Item Definition Object Reference)
   Batch Reverse Retarget(execute=[@n3.then, @n4.then])
   n4 = Get!(execute=@n3.else) -> (then, else, Companion Item Definition:Cosmetic Companion Item Definition Object Reference)
   Show Message Dialog(
    execute=@n4.else
    Title=MakeLiteralText(Value="LOCTEXT(\"Aborted Companion\")")
    Message=MakeLiteralText(Value="LOCTEXT(\"Cannot find Character Definition\")")
    MessageType=Ok
    DefaultValue=No
    MessageCategory=Warning
   )
   n5 = GetSelectedAssets(execute=@n2.else)
   > n6 = Branch(Condition=Is Not Empty(TargetArray=@n5))
   > n7 = Show Message Dialog(
    Title=MakeLiteralText(Value="LOCTEXT(\"No Anim sequence in Anim Browser Selected\")")
    Message=MakeLiteralText(Value="LOCTEXT(\"Do you want to run Reverse Retarget on selected asset in content browser instead?\")")
    MessageType=YesNo
    DefaultValue=No
    MessageCategory=Warning
   )
   > n8 = Switch on EAppReturnType(Selection=@n7)
   Show Message Dialog(
    execute=@n6.else
    Title=MakeLiteralText(Value="LOCTEXT(\"Aborted\")")
    Message=MakeLiteralText(Value="LOCTEXT(\"Please Select anim sequences from anim list\")")
    MessageType=Ok
    DefaultValue=No
    MessageCategory=Warning
   )
   n9 = "On Property Changed (OptionsView)"!() -> (OutputDelegate, then, PropertyName:Name=None) {is=title K2Node_ComponentBoundEvent}
   > Update Variables when property changes!(execute, self, Selection:Name=@n9.PropertyName) -> (then)
   n10 = "On Property Changed (SaveOptionsView)"!() -> (OutputDelegate, then, PropertyName:Name=None) {is=title K2Node_ComponentBoundEvent}
   > Update Variables when property changes(Selection=@n10.PropertyName)
   Batch Reverse Retarget() {is=title K2Node_CustomEvent}
   > n11 = GetSelectedItems(self=Get Anim List View())
   > n12 = Branch(Condition=Is Not Empty(TargetArray=@n11))
   > n13 = Get Array of anim sequences to retarget from List View!(execute, self) -> (then, Source Assets to Retarget:Array of Asset Data Structures) {is=title K2Node_CallFunction}
   Clear(execute=[@n13.then, @n21.then], TargetArray=Get Retargeted animation() -> (Retargeted animation:Array of Strings))
   > Branch(Condition=Is Not Empty(TargetArray:Array of Asset Data Structures=Get Source Assets to Retarget() -> (Source Assets to Retarget:Array of Asset Data Structures)))
   n14 = Get Source Assets to Retarget()
   n15 = "To String (Name)"(InName=Break Asset Data(AssetData="Get (a ref)"(Array:Array of Asset Data Structures=@n14, Dimension 1=0) -> (Output:"Asset Data Structure (by ref)")).PackagePath)
   > n16 = DuplicateAndRetarget!(
    execute
    AssetsToRetarget:Array of Asset Data Structures=Get Source Assets to Retarget()
    SourceMesh:Skeletal Mesh Object Reference=Get Source Skeletal Mesh() -> (Source Skeletal Mesh:Skeletal Mesh Object Reference)
    TargetMesh:Skeletal Mesh Object Reference=Get Target Skeletal Mesh() -> (Target Skeletal Mesh:Skeletal Mesh Object Reference)
    IKRetargetAsset:IKRetargeter Object Reference=Get Reverse IKRetargeter() -> (Reverse IKRetargeter:IKRetargeter Object Reference)
    Search:String=Get Search() -> (Search:String)
    Replace:String=Get Replace() -> (Replace:String)
    Prefix:String=Get Prefix() -> (Prefix:String)
    Suffix:String=Get Suffix() -> (Suffix:String)
    TargetPath:String=SelectString(
     A=Get Custom Path() -> (Custom Path:String)
     B=Append!(A=@n15, B=/ReverseRetargetAnim/) -> (ReturnValue)
     bPickA=Get Use Custom Path() -> (Use Custom Path:bool)
    )
    bUseSourcePath:bool=Get Use Source Path() -> (Use Source Path:bool)
    bIncludeReferencedAssets:bool=Get Include Referenced Assets() -> (Include Referenced Assets:bool)
    bOverwriteExistingFiles:bool=Get Overwrite Existing Files() -> (Overwrite Existing Files:bool)
   ) -> (then, ReturnValue:Array of Asset Data Structures)
   > DelayUntilNextTick!(execute, WorldContextObject:Object Reference) -> (then)
   > n17 = For Each Loop(Array=@n16)
   n18 = Get Retargeted animation()
   SyncBrowserToObjects(
    execute=@n17.Completed
    AssetPaths=Make Array("[0]":String="Get (a ref)"(Array:Array of Strings=@n18, Dimension 1=0) -> (Output:"String (by ref)")) -> (Array:Array of Strings)
   )
   > Branch(Condition=Get Create preview level sequence() -> (Create preview level sequence:bool))
   n19 = Cast To AnimSequence!(execute=@n17.LoopBody, Object=GetAsset(InAssetData=@n17)) -> (then, CastFailed, AsAnim Sequence:Anim Sequence Object Reference)
   > PrintString(InString=Append!(A="\r\n__________", B=GetDisplayName(Object=@n19), C:String=_______________________________________) -> (ReturnValue))
   > AddCurveToAnimSequenceBaseWithValue!(execute, AnimSequenceBase:Anim Sequence Base Object Reference=@n19, CurveName:Name=ReverseRetarget, Value:float=1, bOverwrite:bool=true) -> (then)
   > Add Unique(TargetArray=Get Retargeted animation(), NewItem="To String (Name)"(InName=Break Asset Data(AssetData=@n17).PackagePath))
   n20 = GetSelectedAssetData(execute=@n12.else)
   > n21 = Set Source Assets to Retarget(Source Assets to Retarget:Array of Asset Data Structures=@n20) -> (Output_Get:Array of Asset Data Structures)
   n22 = Asset changed!() -> (OutputDelegate, then, SelectedItem:String, SelectionType:ESelectInfo Enum) {is=title K2Node_CustomEvent}
   n23 = Get Animation_Browser() -> (Animation_Browser:Animation Browser Object Reference)
   > n24 = Set Asset Name(Asset Name:String=GetSelectedOption!(self:"Searchable Combo Box (String) Object Reference"=Get Asset_ComboBox!(self:Animation Browser Object Reference=@n23) -> (Asset_ComboBox:"Searchable Combo Box (String) Object Reference")) -> (ReturnValue:String)) -> (Output_Get:String)
   > n25 = Set Variant Name(Variant Name:String=GetSelectedOption(self=Get Plugin_Combobox!(self:Animation Browser Object Reference=@n23) -> (Plugin_Combobox:"Searchable Combo Box (String) Object Reference"))) -> (Output_Get:String)
   > n26 = AnimatableCreature Find!(
    execute
    self:Animation Browser Object Reference=@n23
    AnimatableCreatures:Array of Animatable Creature Object References=Get AnimatableCreature!(self:Animation Browser Object Reference=@n23) -> (AnimatableCreature:Array of Animatable Creature Object References)
    Asset:String=@n24
    Variant:String=@n25
   ) -> (then, AnimatableCreature:Animatable Creature Object Reference, Exact Match:bool) {is=title K2Node_CallFunction}
   > Set AC(AC:Animatable Creature Object Reference=@n26) -> (Output_Get:Animatable Creature Object Reference)
   > Update Variables when property changes(Selection=AC)
   Event PreConstruct()
   > SetObject(
    self=[
     Get OptionsView() -> (OptionsView:Details View Object Reference)
     Get ReadOnlyView() -> (ReadOnlyView:Details View Object Reference)
     Get SaveOptionsView() -> (SaveOptionsView:Details View Object Reference)
    ]
    NewObject=Self-Reference()
   )
   Event Construct()
   > Load reverse IKR()
   > Find CID and Companions()
   > Find AC files()
   > n27 = Set Anim List View(Anim List View:List View Object Reference=Get ListView!(self:Animation Browser Object Reference=Get Animation_Browser()) -> (ListView:List View Object Reference)) -> (Output_Get:List View Object Reference)
   > n28 = SetSelectionMode!(execute, self:List View Object Reference=@n27, SelectionMode:ESelectionMode Enum=Multi) -> (then)
   n29 = Get Animation_Browser()
   n30 = Branch(
    execute=[@n28.then, @n31.then]
    Condition=Is Empty!(TargetArray:Array of Animatable Creature Object References=Get AnimatableCreature(self=@n29)) -> (ReturnValue:bool)
   )
   > n31 = Delay!(execute, WorldContextObject:Object Reference, Duration=0.5) -> (then)
   n32 = Get Animation_Browser()
   Refresh Assets(execute=@n30.else, self:Animation Browser Object Reference=@n32)
   > n33 = Set Asset Name(Asset Name=GetSelectedOption(self=Get Asset_ComboBox(self=@n32)))
   > n34 = Set Variant Name(Variant Name=GetSelectedOption(self=Get Plugin_Combobox(self=@n32)))
   > n35 = AnimatableCreature Find(self=@n32, AnimatableCreatures=Get AnimatableCreature(self=@n32), Asset=@n33, Variant=@n34)
   > Set AC(AC=@n35)
   > Update Variables when property changes(Selection=AC)
   > Bind Event to On Selection Changed(self:"Searchable Combo Box (String) Object Reference"=Get Asset_ComboBox(self=Get Animation_Browser()), Delegate=@n22) {is=title K2Node_AddDelegate}
   > Update Variables when property changes(Selection=Use Custom IKR)
  ]}
 Find AC files {nodes [
   Find AC files() {is=title K2Node_FunctionEntry}
   > Clear!(
    execute
    TargetMap:Map of Strings to Soft Object Path Structures=Get Found AC file paths() -> (Found AC file paths:Map of Strings to Soft Object Path Structures)
   ) -> (then) {is=title K2Node_CallFunction}
   > n1 = GetAssetsByClass(
    self=GetAssetRegistry()
    ClassPathName=GetSoftClassTopLevelAssetPath(SoftClassReference=/AnimInEdTools/AnimatableCreature/AnimatableCreature.AnimatableCreature_C)
    bSearchSubClasses=false
   )
   > n2 = For Each Loop(Array=@n1)
   Return Node(execute=@n2.Completed)
   > n3 = Split!(
    execute:Exec
    SourceString="To String (Name)"(InName=Break Asset Data(AssetData=@n2).AssetName)
    InStr=AC_
    SearchCase=CaseSensitive
    SearchDir=FromStart
   ) -> (then:Exec, LeftS, RightS, ReturnValue)
   > Branch(Condition=@n3.ReturnValue)
   > Add(TargetMap=Get Found AC file paths(), Key=@n3.RightS, Value=ToSoftObjectPath(InAssetData=@n2))
  ]}
 Find CID and Companions {nodes [
   Find CID and Companions() {is=title K2Node_FunctionEntry}
   > Clear(TargetMap=Get Found CID file paths() -> (Found CID file paths:Map of Strings to Soft Object Path Structures))
   > n1 = GetAssetsByClass(
    self=GetAssetRegistry()
    ClassPathName=GetSoftClassTopLevelAssetPath(SoftClassReference=/Script/FortniteGame.AthenaCharacterItemDefinition)
    bSearchSubClasses=true
   )
   > n2 = For Each Loop(Array=@n1)
   n3 = GetAssetsByClass(
    execute=@n2.Completed
    self=GetAssetRegistry()
    ClassPathName=GetSoftClassTopLevelAssetPath(SoftClassReference=/Script/CosmeticCompanionsRuntime.CosmeticCompanionItemDefinition)
    bSearchSubClasses=true
   )
   > n4 = For Each Loop(Array=@n3)
   Return Node(execute=@n4.Completed)
   > n5 = Split(
    SourceString="To String (Name)"(InName=Break Asset Data(AssetData=@n4).AssetName)
    InStr=Companion_
    SearchCase=CaseSensitive
    SearchDir=FromStart
   )
   > Branch(Condition=@n5.ReturnValue)
   > Add(
    TargetMap=Get Found Companion Def file paths() -> (Found Companion Def file paths:Map of Strings to Soft Object Path Structures)
    Key=@n5.RightS
    Value=ToSoftObjectPath(InAssetData=@n4)
   )
   n6 = "To String (Name)"(InName=Break Asset Data(AssetData=@n2).AssetName)
   n7 = Split(execute=@n2.LoopBody, SourceString=@n6, InStr=Outfit_, SearchCase=CaseSensitive, SearchDir=FromStart)
   > n8 = Branch(Condition=@n7.ReturnValue)
   n9 = ToSoftObjectPath(InAssetData=@n2)
   > Add(TargetMap=Get Found CID file paths(), Key=@n7.RightS, Value=@n9)
   n10 = Split(execute=@n8.else, SourceString=@n6, InStr=NPC_, SearchCase=CaseSensitive, SearchDir=FromStart)
   > n11 = Branch(Condition=@n10.ReturnValue)
   > Add(TargetMap=Get Found CID file paths(), Key=@n10.RightS, Value=@n9)
   n12 = Split(execute=@n11.else, SourceString=@n6, InStr=CID_, SearchCase=CaseSensitive, SearchDir=FromStart)
   > Branch(Condition=@n12.ReturnValue)
   > Add(TargetMap=Get Found CID file paths(), Key=@n12.RightS, Value=@n9)
  ]}
 Find Custom IKR {nodes [
   Find Custom IKR() {is=title K2Node_FunctionEntry}
   > n1 = Branch(Condition=Is Empty!(TargetMap:Map of Strings to Soft Object Path Structures=Get Cached IKRs()) -> (ReturnValue:bool) {is=title K2Node_CallFunction})
   > n2 = Cache IKRs() {is=title K2Node_CallFunction}
   n3 = "For Each Loop (Map)"!(execute:Exec=[@n2.then, @n1.else], MapPin:Map of Strings to Soft Object Path Structures=Get Cached IKRs(), BreakPin:Exec=@n7.then_1) -> (then:Exec, KeyPin:String, ValuePin:Soft Object Path Structure, CompletedPin:Exec)
   n4 = Get Asset Name() -> (Asset Name:String)
   > Branch(Condition=AND Boolean(A=Contains(SearchIn=@n3, Substring=@n4), B=Contains(SearchIn=@n3, Substring=ReverseRetarget, bUseCase=true, bSearchFromEnd=true)))
   > n5 = LoadAsset(AssetPath=BreakSoftObjectPath(InSoftObjectPath=@n3.ValuePin))
   > n6 = Cast To IKRetargeter!(execute, Object=@n5) -> (then, CastFailed, AsIKRetargeter:IKRetargeter Object Reference) {is=title K2Node_DynamicCast}
   > Set Reverse IKRetargeter(Reverse IKRetargeter:IKRetargeter Object Reference=@n6) -> (Output_Get:IKRetargeter Object Reference)
   > n7 = Sequence()
   Return Node!(execute=@n3.CompletedPin, ReturnValue:bool=IsValid(Object=Get Reverse IKRetargeter()))
  ]}
 Get Array of anim sequences to retarget from List View {nodes [
   Get Array of anim sequences to retarget from List View() {is=title K2Node_FunctionEntry}
   > Clear(TargetArray:Array of Anim Sequence Object References=Get Bespoke Animation Sequences() -> (Bespoke Animation Sequences:Array of Anim Sequence Object References)) {is=title K2Node_CallArrayFunction}
   > n1 = GetSelectedItems(self=Get Anim List View())
   > n2 = For Each Loop(Array:Array of Object References=@n1) -> (Array Element:Object Reference)
   Clear(execute=@n2.Completed, TargetArray:Array of Asset Data Structures=Get Source Assets to Retarget())
   > n3 = For Each Loop(Array:Array of Anim Sequence Object References=Get Bespoke Animation Sequences()) -> (Array Element:Anim Sequence Object Reference)
   Return Node!(execute=@n3.Completed, Source Assets to Retarget:Array of Asset Data Structures=Get Source Assets to Retarget())
   > Add Unique(
    TargetArray:Array of Asset Data Structures=Get Source Assets to Retarget()
    NewItem:"Asset Data Structure (by ref)"=CreateAssetData(InAsset=@n3, bAllowBlueprintClass=false)
   )
   n4 = Cast To Object_AnimationBrowser!(execute=@n2.LoopBody, Object=@n2) -> (then, CastFailed, AsObject Animation Browser:Object Animation Browser Object Reference) {is=title K2Node_DynamicCast}
   > n5 = LoadAsset_Blocking(Asset=Get!(self:Object Animation Browser Object Reference=@n4) -> (
     Data_AnimSequence_31_#1:Soft Object Reference
     Data_LevelSequence_32_#2:Soft Object Reference
     Data_CurrentUser_26_#3:String
    ))
   > n6 = Cast To AnimSequence(Object=@n5)
   > Add Unique(
    TargetArray:Array of Anim Sequence Object References=Get Bespoke Animation Sequences()
    NewItem:"Anim Sequence Object Reference (by ref)"=@n6
   )
  ]}
 Load reverse IKR {nodes [
   Load reverse IKR() {is=title K2Node_FunctionEntry}
   > LoadAsset(AssetPath=/CosmeticCompanions/Assets/Quadruped/DoveSlide/Meshes/RTG_Quadruped_ReverseRetarget)
   > LoadAsset(AssetPath=&RTG_Quadruped_ReverseRetarget)
   > Return Node!(execute)
  ]}
 Look For Source SKM from CID {nodes [
   Look For Source SKM from CID() {is=title K2Node_FunctionEntry}
   > n1 = For Each Loop with Break(
    Array:Array of Custom Character Part Soft Object References=Get BaseCharacterParts!(self:Athena Character Item Definition Object Reference=Get CID() -> (CID:Athena Character Item Definition Object Reference)) -> (BaseCharacterParts:Array of Custom Character Part Soft Object References)
    Break=@n6.then
   ) -> (Array Element:Custom Character Part Soft Object Reference)
   n2 = Get!(execute:Exec=@n1.Completed) -> (then:Exec, else:Exec, Source Skeletal Mesh:Skeletal Mesh Object Reference)
   > Return Node!(execute, Success:bool=true)
   Return Node(execute=@n2.else, Success=false)
   n3 = Branch(
    execute=@n1.LoopBody
    Condition=Contains(
     SearchIn=To Soft Object Path!(SoftObjectReference:Soft Object Reference=@n1) -> (ReturnValue_PathString:String)
     Substring=CP_Null_Head
     bUseCase=true
     bSearchFromEnd=true
    )
   )
   n4 = LoadAsset_Blocking(execute=@n3.else, Asset=@n1)
   > n5 = Cast To CustomCharacterPart!(execute, Object=@n4) -> (then, CastFailed, AsCustom Character Part:Custom Character Part Object Reference) {is=title K2Node_DynamicCast}
   > Branch(Condition="Equal (Enum)"(
     A:EFortCustomPartType Enum=GetPartType!(self:Custom Character Part Object Reference=@n5) -> (ReturnValue:EFortCustomPartType Enum=Head)
     B:EFortCustomPartType Enum
    ))
   > n6 = Set Source Skeletal Mesh(Source Skeletal Mesh:Skeletal Mesh Object Reference=GetSkeletalMesh!(self:Custom Character Part Object Reference=@n5) -> (ReturnValue:Skeletal Mesh Object Reference)) -> (Output_Get:Skeletal Mesh Object Reference)
  ]}
 Look For Source SKM from Companion Def {nodes [
   Look For Source SKM from Companion Def() {is=title K2Node_FunctionEntry}
   > n1 = For Each Loop(Array:Array of Custom Character Part Object References=GetCharacterParts!(self:Athena Character Part Item Definition Object Reference=Get Companion Item Definition() -> (Companion Item Definition:Cosmetic Companion Item Definition Object Reference)) -> (ReturnValue:Array of Custom Character Part Object References)) -> (Array Element:Custom Character Part Object Reference)
   n2 = Get(execute=@n1.Completed)
   > Return Node(Success=true)
   Return Node(execute=@n2.else, Success=false)
   n3 = Branch(execute=@n1.LoopBody, Condition=Contains(SearchIn=GetDisplayName(Object=@n1), Substring=CP_Null_Head, bUseCase=true, bSearchFromEnd=true))
   Branch(
    execute=@n3.else
    Condition="Equal (Enum)"(A=GetCharacterPartType!(self:Custom Character Part Object Reference=@n1) -> (ReturnValue:EFortCustomPartType Enum=Head))
   )
   > Set Source Skeletal Mesh(Source Skeletal Mesh=GetSkeletalMesh(self=@n1))
   Get MainSkeletalMesh!(self:Cosmetic Companion Item Definition Object Reference=Get Companion Item Definition()) -> (MainSkeletalMesh:Skeletal Mesh Soft Object Reference)
  ]}
 Set Custom path {nodes [
   Set Custom path() {is=title K2Node_FunctionEntry}
   > n1 = Get!(execute) -> (then, else, AC:Animatable Creature Object Reference)
   n2 = Get Path Parts Map!(
    Asset:String=Get Asset!(self:Animatable Creature Object Reference=@n1) -> (Asset)
    Animation Set:String
    Animation Name:String
    Descriptor:String
    Direction:String
    Subfolder:String="Get (a ref)"(
     Array=Get Subfolders!(self:Animatable Creature Object Reference=@n1) -> (Subfolders:Array of Strings) {is=title K2Node_CallFunction}
     Dimension 1=0
    )
    Variation:String
    Plugin:String=Get Variant!(self:Animatable Creature Object Reference=@n1) -> (Variant)
    __WorldContext:Object Reference
   ) -> (Output Path Parts Map:Map of Strings to Strings) {is=title K2Node_CallFunction}
   n3 = Get Naming Convention!(self:Animatable Creature Object Reference=@n1) -> (ReturnValue:EU Animation Naming Convention Object Reference) {is=title K2Node_CallFunction}
   n4 = Get Animation Path!(self:Animatable Creature Object Reference=@n1) -> (Animation Path:String)
   > Set Custom Path(Custom Path:String=Append!(
     A=Get Path!(
      self:EU Animation Naming Convention Object Reference=@n3
      Token Map:Map of Strings to Strings=@n2
      Path Pattern:String=@n4
      Remove Spaces On Name:bool=true
     ) -> (File Path:String, Folder Path:String, File Name:String) {is=title K2Node_CallFunction}.Folder Path
     B=/ReverseRetargetAnim/
    ) -> (ReturnValue)) -> (Output_Get:String)
   > Return Node!(execute)
  ]}
 Update Variables when property changes {nodes [
   n1 = Update Variables when property changes!() -> (then, Selection:Name) {is=title K2Node_FunctionEntry}
   > n2 = Switch on Name!(execute:Exec, Selection:Name=@n1) -> (Default:Exec, AC:Exec, Companion Item Definition:Exec, CID:Exec, Use Custom IKR:Exec) {is=title K2Node_SwitchName, pinNames [AC, Companion Item Definition, CID, Use Custom IKR]}
   n3 = Sequence(execute=[@n2.AC, @n2.CID, @n2.Companion Item Definition])
   > n4 = Sequence!(execute) -> (then_0, then_1, then_2:Exec)
   > Get()
   n5 = Find(
    TargetMap:Map of Strings to Soft Object Path Structures=Get Found CID file paths()
    Key=Append!(A=Get Asset Name(), B=_, C:String=Get Variant Name() -> (Variant Name:String)) -> (ReturnValue)
   ) -> (Value:Soft Object Path Structure)
   > n6 = Branch(Condition=@n5.ReturnValue)
   > n7 = Set Found AC File path(Found AC File path:Soft Object Path Structure=@n5) -> (Output_Get:Soft Object Path Structure)
   n8 = LoadAsset(execute=[@n12.then, @n7.then], AssetPath=Get!() -> (Found AC File path_PathString:String))
   > n9 = Cast To AthenaCharacterItemDefinition!(execute, Object=@n8) -> (then, CastFailed, AsAthena Character Item Definition:Athena Character Item Definition Object Reference) {is=title K2Node_DynamicCast}
   > Set CID(CID:Athena Character Item Definition Object Reference=@n9) -> (Output_Get:Athena Character Item Definition Object Reference)
   > Set Companion Item Definition(Companion Item Definition:Cosmetic Companion Item Definition Object Reference) -> (Output_Get:Cosmetic Companion Item Definition Object Reference)
   > Look For Source SKM from CID!(execute, self) -> (then, Success:bool) {is=title K2Node_CallFunction}
   n10 = Cast To CosmeticCompanionItemDefinition!(execute=@n9.CastFailed, Object=@n8) -> (then, CastFailed, AsCosmetic Companion Item Definition:Cosmetic Companion Item Definition Object Reference) {is=title K2Node_DynamicCast}
   > Set Companion Item Definition(Companion Item Definition=@n10)
   > Set CID()
   > Look For Source SKM from Companion Def!(execute, self) -> (then, Success:bool) {is=title K2Node_CallFunction}
   n11 = Find(TargetMap=Get Found Companion Def file paths(), Key=Get Asset Name())
   Branch(execute=@n6.else, Condition=@n11.ReturnValue)
   > n12 = Set Found AC File path(Found AC File path=@n11)
   Set Custom path(execute=@n4.then_2) {is=title K2Node_CallFunction}
   n13 = Get!(execute:Exec=@n4.then_1) -> (then:Exec, else:Exec, Companion Item Definition:Cosmetic Companion Item Definition Object Reference)
   n14 = Get MainSkeletalMesh(self=@n13)
   > n15 = Branch(Condition=IsValidSoftObjectReference(SoftObjectReference=@n14))
   > n16 = LoadAsset_Blocking(Asset=@n14)
   > n17 = Cast To SkeletalMesh!(execute, Object=@n16) -> (then, CastFailed, AsSkeletal Mesh:Skeletal Mesh Object Reference) {is=title K2Node_DynamicCast}
   > Set Target Skeletal Mesh(Target Skeletal Mesh:Skeletal Mesh Object Reference=@n17) -> (Output_Get:Skeletal Mesh Object Reference)
   n18 = LogString(execute=@n15.else, InString=NO VALID SKM REF)
   n19 = Get IK Retarget Controller!(execute=[@n18.then, @n20.then], InRetargeterAsset:IKRetargeter Object Reference=Get Reverse IKRetargeter()) -> (then, ReturnValue:IKRetargeter Controller Object Reference) {is=title K2Node_CallFunction}
   > Set Target Skeletal Mesh(Target Skeletal Mesh=GetPreviewMesh!(self:IKRetargeter Controller Object Reference=@n19, SourceOrTarget:ERetargetSourceOrTarget Enum=Target) -> (ReturnValue:Skeletal Mesh Object Reference))
   n20 = LogString(execute=@n13.else, InString=NO COID)
   Return Node(execute=@n3.then_1)
   n21 = Branch(execute=@n2.Use Custom IKR, Condition=Get Use Custom IKR() -> (Use Custom IKR:bool))
   > n22 = Find Custom IKR!(execute, self) -> (then, ReturnValue:bool) {is=title K2Node_CallFunction}
   > n23 = Branch(Condition=@n22)
   n24 = Get AC() -> (AC:Animatable Creature Object Reference)
   n25 = Branch(execute=[@n23.else, @n21.else], Condition=Contains(SearchIn=Get Variant(self=@n24), Substring=Quadruped))
   > n26 = LoadAsset(AssetPath=&RTG_Quadruped_ReverseRetarget)
   > n27 = Cast To IKRetargeter(Object=@n26)
   > Set Reverse IKRetargeter(Reverse IKRetargeter=@n27)
   n28 = LoadAsset(execute=@n25.else, AssetPath=/CosmeticCompanions/Assets/Biped/RoseBruin/Meshes/RTG_Biped_ReverseRetarget)
   > n29 = Cast To IKRetargeter(Object=@n28)
   > Set Reverse IKRetargeter(Reverse IKRetargeter=@n29)
  ]}
]
variables [
 AC {type=AnimatableCreature_C, default=null}
 Anim List View {type=ListView, default=null}
 Asset Name {type=string, default=""}
 Bespoke Animation Sequences {type="array<AnimSequence>", default []}
 Cached IKRs {type="map<SoftObjectPath>", default {}}
 CID {type=AthenaCharacterItemDefinition, default=null}
 Companion Item Definition {type=CosmeticCompanionItemDefinition, default=null}
 Create preview level sequence {type=boolean}
 Custom Path {type=string, default=_ReverseRetarget}
 Found AC file paths {type="map<SoftObjectPath>", default {}}
 Found CID file paths {type="map<SoftObjectPath>", default {}}
 Found Companion Def file paths {type="map<SoftObjectPath>", default {}}
 Include Referenced Assets {type=boolean, default=true}
 Overwrite Existing Files {type=boolean, default=true}
 Prefix {type=string, default=""}
 Replace {type=string, default=""}
 Retargeted animation {type="array<string>", default []}
 Reverse IKRetargeter {type=IKRetargeter, default=null}
 Search {type=string, default=""}
 Source Assets to Retarget {type="array<AssetData>", default []}
 Source CID Folder {type=string, default=""}
 Source Skeletal Mesh {type=SkeletalMesh, default=null}
 Suffix {type=string, default=_ReverseRetarget}
 Target Skeletal Mesh {type=SkeletalMesh, default=null}
 Use Custom IKR {type=boolean}
 Use Custom Path {type=boolean}
 Use Source Path {type=boolean}
 Variant Name {type=string, default=""}
]
widgets {
 info {rootWidgetClass=&CanvasPanel, widgetCount=9}
 widgets [
  CanvasPanel_0 {widgetClassPath=&CanvasPanel, widget=CanvasPanel_0}
  VerticalBox_26 {
   widgetClassPath=&VerticalBox
   parent=CanvasPanel_0
   widget=VerticalBox_26
   slot {class=CanvasPanelSlot, properties {layoutData {alignment=<0 0>, anchors {maximum=<1 1>, minimum=<0 0>}, offsets=<ltrb 0 0 0 0>}}}
  }
  EditorUtilityScrollBox_57 {
   widgetClassPath=&EditorUtilityScrollBox
   parent=VerticalBox_26
   widget=EditorUtilityScrollBox_57
   properties {alwaysShowScrollbar=true}
   slot {class=VerticalBoxSlot, properties {size {sizeRule=Fill, value=1}}}
  }
  Animation_Browser {
   widgetClassPath=/AnimInEdTools/AnimationBrowser/Animation_Browser
   bIsVariable=true
   parent=EditorUtilityScrollBox_57
   widget=Animation_Browser
   slot {class=ScrollBoxSlot, properties {size {sizeRule=Fill, value=1}}}
  }
  OptionsView {
   widgetClassPath=&DetailsView
   parent=EditorUtilityScrollBox_57
   widget=OptionsView
   properties {categoriesToShow [Detail View]}
   slot {class=ScrollBoxSlot}
  }
  ReadOnlyView {
   widgetClassPath=&DetailsView
   parent=EditorUtilityScrollBox_57
   widget=ReadOnlyView
   properties {bIsEnabled=false, visibility=Collapsed, categoriesToShow [Read Only Var]}
   slot {class=ScrollBoxSlot}
  }
  SaveOptionsView {
   widgetClassPath=&DetailsView
   parent=EditorUtilityScrollBox_57
   widget=SaveOptionsView
   properties {categoriesToShow [Save Options]}
   slot {class=ScrollBoxSlot}
  }
  ExecButton {
   widgetClassPath=&EditorUtilityButton
   parent=EditorUtilityScrollBox_57
   widget=ExecButton
   properties {toolTipText=Select animation sequences from the anim list and run}
   slot {class=ScrollBoxSlot}
  }
  TextBlock_28 {
   widgetClassPath=&TextBlock
   parent=ExecButton
   widget=TextBlock_28
   properties {text=Reverse Retarget Animation}
   slot {class=ButtonSlot}
  }
 ]
}
